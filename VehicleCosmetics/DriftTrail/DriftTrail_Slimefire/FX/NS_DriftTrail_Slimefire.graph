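# Niagara system for the Slimefire vehicle drift trail cosmetic, attached to wheel sockets sampled from the vehicle's skeletal mesh.
# Four emitters layer burst-spawned sphere sprites and meshes, short-lived shrinking sprites thrown backward, quality-scaled two-ribbon trails with five renderers, and mesh wraps scaled to the wheels.
# Each emitter applies per-color painted-variant overrides and masks particles per tire through FN_DisableTireFX.
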NiagaraSystem /VehicleCosmetics/DriftTrail/DriftTrail_Slimefire/FX/NS_DriftTrail_Slimefire
dependencies {
 usedDataInterfaces [&NiagaraDataInterfaceColorCurve, &NiagaraDataInterfaceCurve, &NiagaraDataInterfaceSkeletalMesh]
 usedDynamicInputs []
 usedModules [
  &self:SystemSpawnScript.SystemScriptSource.SystemScriptGraph.NiagaraNodeAssignment_0.SetVariables_#1
  &FN_UnpackVehicleFlags
  /VehicleCosmetics/FX/Niagara/Modules/VC_SetVehicleSpeedAttribs
  &SystemState
  /Game/Effects/Niagara/Modules/Vehicle/FN_VehicleFlags_Cheatcodes
 ]
}
emitters {
 NE_SF_Sphere {
  emitterSpawnScript [/SetVariables(
    Emitter.Curve for Colors=DataInterface() {propertyValues {
      ExposedName=MainTrailGradient
      bExposeCurve=true
      AlphaCurve {keys [<at 0.506 1>]}
      BlueCurve {keys [<at 0.8547 0>,<at 0.9718 0>]}
      GreenCurve {keys [<at 0.8547 0.0514>,<at 0.9718 0.7134>]}
      RedCurve {keys [<at 0.8547 1.5>,<at 0.9718 1>]}
     }}
    Emitter.FloatIsPainted=HlslExpression() {hlslExpression="User.ColorVariantID == 0 ? 0 : 1"}
   ) {isSetParameters=true}]
  emitterUpdateScript [
   &EmitterState(
    Inactive Response=::NewEnumerator1
    Life Cycle Mode=::NewEnumerator1
    Loop Behavior=::NewEnumerator1
    Loop Duration Mode=::NewEnumerator1
    UseLoopDelay=false
   )
   &SpawnBurst_Instantaneous(Spawn Count=2, Use Spawn Probability=false, UseLoopCountLimit=false)
   &SpawnBurst_Instantaneous(Spawn Count=2, Spawn Group=1, Use Spawn Probability=false, UseLoopCountLimit=false) {module=SpawnBurst_Instantaneous001}
   &SpawnBurst_Instantaneous(Spawn Count=2, Spawn Group=2, Use Spawn Probability=false, UseLoopCountLimit=false) {module=SpawnBurst_Instantaneous002}
   &SpawnBurst_Instantaneous(Spawn Count=2, Spawn Group=3, Use Spawn Probability=false, UseLoopCountLimit=false) {module=SpawnBurst_Instantaneous003}
   &SpawnRate(
    SpawnRate=/VehicleCosmetics/FX/Niagara/Modules/VC_Mask_Floats(
     FrontEnd Value:NiagaraFloat=200
     blsFrontEndPreview:NiagaraBool=false
     SpeedMin Value:NiagaraFloat=120
     SpeedMax Value:NiagaraFloat=180
     Speed Alpha:NiagaraFloat=$System.Speed_Alpha
    )
    Use Spawn Probability=false
   ) {enabled=false}
  ]
  particleSpawnScript [
   &VC_PaintedVariantsOverride(
    Black=<rgba 0.036 0.0382 0.045 1>
    BurntSienna=<rgba 0.1 0.0235 0.01 1>
    Cobalt=<rgba 0.025 0.0646 0.5 1>
    Crimson=<rgba 0.4479 0.0403 0.0743 1>
    DefaultColor=<rgba 0.8647 0.2514 0.047 1>
    ForestGreen=<rgba 0.15 0.6 0.06 1>
    Grey=<rgba 0.08 0.0867 0.1 1>
    Lime=<rgba 0.2917 0.5 0 1>
    Orange=<rgba 0.4 0.0928 0.032 1>
    Pink=<rgba 0.3656 0.1125 0.45 1>
    Purple=<rgba 0.133 0.07 0.7 1>
    R1 Black:LinearColor=<rgba 0.15 0.15 0.15 1>
    R1 BurntSienna:LinearColor=<rgba 0.4583 0.1362 0.0549 1>
    R1 Cobalt:LinearColor=<rgba 0.3 0.567 1 1>
    R1 Crimson:LinearColor=<rgba 0.8 0.2625 0.2771 1>
    R1 DefaultColor=<rgba 1 1 0.0369 1>
    R1 ForestGreen:LinearColor=<rgba 0.2868 2 0.2 1>
    R1 Grey:LinearColor=<rgba 0.526 0.526 0.526 1>
    R1 Lime:LinearColor=<rgba 1.1 2 0.2 1>
    R1 Orange:LinearColor=<rgba 1.5 0.572 0.075 1>
    R1 Pink:LinearColor=<rgba 1 0.5208 1 1>
    R1 Purple:LinearColor=<rgba 0.4417 0.125 2.5 1>
    R1 Saffron:LinearColor=<rgba 1 0.8517 0.3646 1>
    R1 SkyBlue:LinearColor=<rgba 0.2708 0.6962 1 1>
    R1 TitaniumWhite:LinearColor=<rgba 1.5 1.5 1.5 1>
    R2 Black:LinearColor=<rgba 0.024 0.024 0.024 1>
    R2 BurntSienna:LinearColor=<rgba 0.25 0.0566 0.0225 1>
    R2 Cobalt:LinearColor=<rgba 0.0492 0.1263 0.82 1>
    R2 Crimson:LinearColor=<rgba 0.5573 0.0056 0.0205 1>
    R2 DefaultColor:LinearColor=<rgba 1.1953 1.1953 1.1953 1>
    R2 ForestGreen:LinearColor=<rgba 0.0101 0.5052 0.0351 1>
    R2 Grey:LinearColor=<rgba 0.18 0.185 0.2 1>
    R2 Lime:LinearColor=<rgba 0.3294 0.6458 0.0129 1>
    R2 Orange:LinearColor=<rgba 0.85 0.1558 0.0085 1>
    R2 Pink:LinearColor=<rgba 0.8 0.2 0.7 1>
    R2 Purple:LinearColor=<rgba 0.1733 0.048 0.8 1>
    R2 Saffron:LinearColor=<rgba 1 0.7625 0.05 1>
    R2 SkyBlue:LinearColor=<rgba 0.11 0.6875 1.1 1>
    R2 TitaniumWhite:LinearColor=<rgba 0.65 0.65 0.65 1>
    Saffron=<rgba 0.4 0.31 0.04 1>
    SkyBlue=<rgba 0.0175 0.1839 0.35 1>
    TitaniumWhite=<rgba 0.4 0.4 0.4 1>
    Use R1 Painted Overrides=true
    Use R2 Painted Overrides=true
   )
   &InitializeParticle(
    Color Mode=::NewEnumerator0
    Lifetime=&SelectFloatFromArray(Direct Array Index=&SpawnGroup(), Float Selection Array=DataInterface() {propertyValues {FloatData [1,1,1,1]}})
    Mesh Scale Mode=::NewEnumerator3
    Mesh Uniform Scale=/VehicleCosmetics/FX/Niagara/Modules/MergeEmitters/VC_FloatBySpawnGroup(
     SpawnGroup or ExecIndex:NiagaraInt32=0
     CustomInt:NiagaraInt32=0
     MaximumSpawnGroups:NiagaraInt32=4
     Float SG 0:NiagaraFloat=0
     Float SG 1:NiagaraFloat=0
     Float SG 2:NiagaraFloat=1.1
     Float SG 3:NiagaraFloat=0
     Float SG 4:NiagaraFloat=0
     Float SG 5:NiagaraFloat=0
    )
    Sprite Size Mode=::NewEnumerator3
    Uniform Sprite Size=&VC_FloatBySpawnGroup(
     SpawnGroup or ExecIndex:NiagaraInt32=0
     CustomInt:NiagaraInt32=0
     MaximumSpawnGroups:NiagaraInt32=4
     Float SG 0:NiagaraFloat=19
     Float SG 1:NiagaraFloat=19
     Float SG 2:NiagaraFloat=0
     Float SG 3:NiagaraFloat=0
     Float SG 4:NiagaraFloat=0
     Float SG 5:NiagaraFloat=0
    )
    UsePositionOffset=false
    Write Material Random=true
   )
   &PartitionParticles()
   &SkeletalMeshLocation(
    Bone / Socket Index=&Add_Integer(A=$Particles.PartitionParticles.Partition, B=4)
    Bone / Socket Sampling Mode=::NewEnumerator5
    Interpolate Mesh Sampling=false
    MaskSpawnGroup=false
    Mesh Sampling Type=::NewEnumerator6
    Sampled Position Offset=&AddVectorToPosition(Position=<0 0 0>, Vector=$User.offset)
    Skeletal Mesh=$System.SkeletalMesh
   )
   &AddVelocity(Velocity=&VC_Mask_Vectors(
     FrontEnd Value:Vector3f=<-1500 0 0>
     blsFrontEndPreview:NiagaraBool=false
     SpeedMin Value:Vector3f=<-800 0 0>
     SpeedMax Value:Vector3f=<-1 0 0>
     Speed Alpha:NiagaraFloat=$System.Speed_Alpha
    )) {enabled=false}
   &CameraOffset(Camera Offset Amount=&VC_FloatBySpawnGroup(
     SpawnGroup or ExecIndex:NiagaraInt32=0
     CustomInt:NiagaraInt32=0
     MaximumSpawnGroups:NiagaraInt32=4
     Float SG 0:NiagaraFloat=0
     Float SG 1:NiagaraFloat=0
     Float SG 2:NiagaraFloat=60
     Float SG 3:NiagaraFloat=0
     Float SG 4:NiagaraFloat=0
     Float SG 5:NiagaraFloat=0
    )) {enabled=false}
  ]
  particleUpdateScript [
   &FN_DisableTireFX(
    Mask Index Attribute:NiagaraInt32=$Particles.PartitionParticles.Partition
    Mask Index FL=2
    Mask Index FR=3
    Mask Index RL=0
    Mask Index RR=1
   )
   &ParticleState(Kill Particles When Lifetime Has Elapsed=false, Loop Particles Lifetime=true)
   &InheritSourceMovement(ApplyForce=false, ApplyPosition=true, ApplyPositionFalloff=false) {enabled=false}
   &SolveForcesAndVelocity(Clamp Velocity=false, Limit Acceleration=false)
   /SetVariables(Particles.Visibility:NiagaraInt32=&SpawnGroup()) {isSetParameters=true}
  ]
  rendererClasses [&NiagaraSpriteRendererProperties, &NiagaraMeshRendererProperties]
  renderers [
   {rendererClass=&NiagaraSpriteRendererProperties}
   {rendererClass=&NiagaraMeshRendererProperties, rendererIndex=1}
   {rendererClass=&NiagaraSpriteRendererProperties, rendererIndex=2}
  ]
 }
 NE_SF_Sprite {
  emitterUpdateScript [
   &EmitterState()
   &SpawnRate(Spawn Probability=0.7, SpawnRate=30, Use Spawn Probability=true)
  ]
  particleSpawnScript [
   &VC_PaintedVariantsOverride(
    Black=<rgba 0.2091 0.2091 0.2091 5>
    BurntSienna=<rgba 0.3741 0.0955 0.0207 5>
    Cobalt=<rgba 0.0875 0.2372 1 5>
    Crimson=<rgba 0.5926 0 0.0161 5>
    DefaultColor=<rgba 3 1 0 2.5>
    ForestGreen=<rgba 0 0.5703 0.0288 5>
    Grey=<rgba 0.3028 0.3028 0.3028 5>
    Lime=<rgba 0.3855 0.771 0 5>
    Orange=<rgba 0.9717 0.2574 0 5>
    Pink=<rgba 1 0.3542 0.7652 5>
    Purple=<rgba 0.5056 0.1615 1 5>
    Saffron=<rgba 1 0.6902 0 5>
    SkyBlue=<rgba 0 0.6169 0.9271 5>
    TitaniumWhite=<rgba 0.8698 0.8698 0.8698 5>
   )
   &InitializeParticle(
    Color Mode=::NewEnumerator0
    Lifetime=&RandomRangeFloat(Minimum=0.25, Maximum=0.35)
    Sprite Size Mode=::NewEnumerator1
    Uniform Sprite Size Max=25
    Uniform Sprite Size Min=20
    UsePositionOffset=false
    Write Material Random=true
   )
   &PartitionParticles()
   &SkeletalMeshLocation(
    Bone / Socket Index=&Add_Integer(A=$Particles.PartitionParticles.Partition, B=4)
    Bone / Socket Sampling Mode=::NewEnumerator5
    Interpolate Mesh Sampling=false
    MaskSpawnGroup=false
    Mesh Sampling Type=::NewEnumerator6
    Sampled Position Offset=&AddVectorToPosition(Position=<-25 0 0>, Vector=$User.offset)
    Skeletal Mesh=$System.SkeletalMesh
   )
   &AddVelocity(Velocity=<-1000 0 0>)
   &ShapeLocation(MaskSpawnGroup=false, Override Seed=false, Sphere Radius=18, UseOwnerScale=true)
   &Color(Color=<rgba 1 0.5 0.05 1>) {enabled=false}
  ]
  particleUpdateScript [
   &FN_DisableTireFX(
    Mask Index Attribute:NiagaraInt32=$Particles.PartitionParticles.Partition
    Mask Index FL=2
    Mask Index FR=3
    Mask Index RL=0
    Mask Index RR=1
   )
   &ParticleState()
   &ScaleSpriteSize(Uniform Curve Sprite Scale=DataInterface() {propertyValues {Curve {
       defaultValue=0
       keys [{interpMode=RCIM_Cubic, tangentMode=RCTM_User, time=0, value=1}, {interpMode=RCIM_Cubic, tangentMode=RCTM_User, time=1, value=0}]
      }}})
   &SolveForcesAndVelocity(Clamp Velocity=false, Limit Acceleration=false)
   &ScaleColor(
    Scale Alpha=&FloatFromCurve(
     FloatCurve=DataInterface() {propertyValues {Curve {
        defaultValue=0
        keys [
         {tangentMode=RCTM_User, time=0, value=1}
         {arriveTangent=-0.0405, interpMode=RCIM_Cubic, leaveTangent=-0.0405, tangentMode=RCTM_User, time=0.7487, value=1}
         {arriveTangent=-8.0089, interpMode=RCIM_Cubic, leaveTangent=-8.0089, tangentMode=RCTM_User, time=1, value=0}
        ]
       }}}
     Scale Curve=4
    )
    ScaleA=true
    ScaleRGB=true
   ) {enabled=false}
  ]
  rendererClasses [&NiagaraSpriteRendererProperties]
  renderers [{rendererClass=&NiagaraSpriteRendererProperties}]
 }
 NE_SF_Trail {
  emitterSpawnScript [/SetVariables(
    Emitter.Curve for Colors=DataInterface() {propertyValues {
      ExposedName=MainTrailGradient
      bExposeCurve=true
      AlphaCurve {keys [<at 0.9412 1>]}
      BlueCurve {keys [<at 0.0949 0.1787>,<at 0.9687 0>]}
      GreenCurve {keys [<at 0.0949 0>,<at 0.9687 0.9333>]}
      RedCurve {keys [<at 0.0949 3>,<at 0.9687 2>]}
     }}
    Emitter.bool:NiagaraBool=&InvertBool(Boolean:NiagaraBool=$User.bIsPainted)
   ) {isSetParameters=true}]
  emitterUpdateScript [
   &EmitterState()
   &VC_RateQualityHandler(Epic Spawn Rate=90, High Spawn Rate=60, Low Spawn Rate=30, Medium Spawn Rate=45, Use Custom Rate?=true)
   &VC_SpawnRibbon_v1(Number of Ribbons=2, Spawn Time=-1, SpawnRate=$Emitter.Spawn Rate)
  ]
  particleSpawnScript [
   &VC_PaintedVariantsOverride(
    Black=<rgba 0.022 0.022 0.022 1>
    BurntSienna=<rgba 0.12 0.0264 0.012 1>
    Cobalt=<rgba 0.025 0.0765 0.5 1>
    Crimson=<rgba 0.3646 0 0.0143 1>
    DefaultColor=&ColorFromCurve(ColorCurve=DataInterface() {propertyValues {
       AlphaCurve {keys [<at 0 1 RCIM_Cubic>,<at 1 0 RCIM_Cubic>]}
       BlueCurve {keys [<at 0 1 RCIM_Cubic>,<at 1 1 RCIM_Cubic>]}
       GreenCurve {keys [<at 0 1 RCIM_Cubic>,<at 1 1 RCIM_Cubic>]}
       RedCurve {keys [<at 0 1 RCIM_Cubic>,<at 1 1 RCIM_Cubic>]}
      }})
    ForestGreen=<rgba 0 0.2396 0.0121 1>
    Grey=<rgba 0.13 0.13 0.13 1>
    Lime=<rgba 0.2 0.4 0 1>
    Orange=<rgba 0.5124 0.0832 0.0307 1>
    Pink=<rgba 0.401 0.1063 0.3819 1>
    Purple=<rgba 0.0565 0.0208 0.45 1>
    R1 Black:LinearColor=<rgba 0.0429 0.0429 0.045 1>
    R1 BurntSienna:LinearColor=<rgba 0.18 0.0321 0.0018 1>
    R1 Cobalt:LinearColor=<rgba 0.035 0.076 0.7 1>
    R1 Crimson:LinearColor=<rgba 0.4792 0 0.005 1>
    R1 DefaultColor=<rgba 0.1009 0.6542 0.8802 1>
    R1 ForestGreen:LinearColor=<rgba 0.0244 0.35 0 1>
    R1 Grey:LinearColor=<rgba 0.136 0.146 0.16 1>
    R1 Lime:LinearColor=<rgba 0.3567 0.8 0.04 1>
    R1 Orange:LinearColor=<rgba 0.75 0.1375 0.015 1>
    R1 Pink:LinearColor=<rgba 0.7 0.14 0.6533 1>
    R1 Purple:LinearColor=<rgba 0.1432 0.0234 0.78 1>
    R1 Saffron:LinearColor=<rgba 0.7656 0.5944 0.0319 1>
    R1 SkyBlue:LinearColor=<rgba 0.0747 0.5317 0.7969 1>
    R1 TitaniumWhite:LinearColor=<rgba 0.6354 0.6354 0.6354 1>
    R2 Black:LinearColor=<rgba 0.0885 0.0885 0.0885 1>
    R2 BurntSienna:LinearColor=<rgba 0.3613 0.1119 0.0437 1>
    R2 Cobalt:LinearColor=<rgba 0.2275 0.3071 0.91 1>
    R2 Crimson:LinearColor=<rgba 0.6823 0 0.0185 1>
    R2 DefaultColor:LinearColor=<rgba 1.3 0.5993 0.299 1>
    R2 ForestGreen:LinearColor=<rgba 0 0.5 0.0253 1>
    R2 Grey:LinearColor=<rgba 0.3125 0.3125 0.3125 1>
    R2 Lime:LinearColor=<rgba 0.4083 0.7 0.1167 1>
    R2 Orange:LinearColor=<rgba 0.6823 0.3268 0.1365 1>
    R2 Pink:LinearColor=<rgba 0.7 0.315 0.7 1>
    R2 Purple:LinearColor=<rgba 0.2628 0.166 0.8854 1>
    R2 Saffron:LinearColor=<rgba 0.8333 0.6389 0 1>
    R2 SkyBlue:LinearColor=<rgba 0 0.4852 0.7292 1>
    R2 TitaniumWhite:LinearColor=<rgba 0.6615 0.6615 0.6615 1>
    R3 Black:LinearColor=<rgba 0.1 0.1 0.1 10>
    R3 BurntSienna:LinearColor=<rgba 0.5312 0.1356 0.0437 2>
    R3 Cobalt:LinearColor=<rgba 0.1906 0.1324 1 2>
    R3 Crimson:LinearColor=<rgba 0.6875 0 0.0116 5>
    R3 DefaultColor:LinearColor=<rgba 0.2524 5.6077 6 1>
    R3 ForestGreen:LinearColor=<rgba 0 0.35 0.0177 10>
    R3 Grey:LinearColor=<rgba 0.18 0.18 0.18 10>
    R3 Lime:LinearColor=<rgba 0.55 1.1 0 20>
    R3 Orange:LinearColor=<rgba 2 0.3333 0 7>
    R3 Pink:LinearColor=<rgba 1 0.25 0.7273 10>
    R3 Purple:LinearColor=<rgba 0.1667 0 1 20>
    R3 Saffron:LinearColor=<rgba 1.1 0.8433 0 5>
    R3 SkyBlue:LinearColor=<rgba 0 0.754 1.5 5>
    R3 TitaniumWhite:LinearColor=<rgba 1.5 1.5 1.5 2>
    Saffron=<rgba 0.599 0.4592 0 1>
    SkyBlue=<rgba 0.0395 0.4179 0.6867 1>
    TitaniumWhite=<rgba 0.408 0.4401 0.4896 1>
    Use R1 Painted Overrides=true
    Use R2 Painted Overrides=true
    Use R3 Painted Overrides=true
   )
   &VC_InitializeRibbon_v1(
    Color=$Particles.Color
    Custom Offset=true
    Enable Skeletal Position=true
    Lifetime=0.75
    Local Vector Offset=&AddVectorToPosition(Position=<25 0 0>, Vector=$User.offset)
    Offset Amount=1
    Ribbon Width=50
   )
   &AddVelocity(Velocity=<-1000 0 0>) {module=AddVelocity001}
  ]
  particleUpdateScript [
   &FN_DisableTireFX(Mask Index Attribute:NiagaraInt32=$Particles.SpawnGroup, Mask Index FL=2, Mask Index FR=3, Mask Index RL=0, Mask Index RR=1)
   &ParticleState()
   &InheritSourceMovement(
    Applied Position Scale=&Multiply_VectorByFloat(
     Vector=<1 1 1>
     Float=&FloatFromCurve(FloatCurve=DataInterface() {propertyValues {Curve {keys [<at 0.3 1 RCIM_Cubic>, {arriveTangent=-3.0131, interpMode=RCIM_Cubic, leaveTangent=-3.0131, tangentMode=RCTM_User, time=1, value=0}]}}})
    )
    ApplyForce=false
    ApplyPosition=true
    ApplyPositionFalloff=false
   )
   &SolveForcesAndVelocity(Clamp Velocity=false, Limit Acceleration=false)
   &VC_UpdateRibbon_v1(Enable Skeletal Position=true)
   &ScaleColor(
    Scale Alpha=&FloatFromCurve(FloatCurve=DataInterface() {propertyValues {Curve {
        defaultValue=0.2358
        keys [
         {arriveTangent=3.2844, interpMode=RCIM_Cubic, leaveTangent=3.2844, tangentMode=RCTM_User, time=0.08, value=0}
         {arriveTangent=0.0427, interpMode=RCIM_Cubic, leaveTangent=0.0206, tangentMode=RCTM_Break, time=0.17, value=1}
         {tangentMode=RCTM_User, time=1, value=0}
        ]
       }}})
    ScaleA=true
    ScaleRGB=false
   )
   /SetVariables(
    Particles.PosOutsideFire:NiagaraPosition=&AddVectorToPosition(Position=$Particles.Position, Vector=/VehicleCosmetics/FX/Niagara/Modules/VC_ConvertVectorToWorld(Vector:Vector3f=<0 0 0>))
    Particles.ScalePainted:LinearColor=&VC_DynamicScaleColor(
     Color to Scale:LinearColor=$Particles.Initial.Color
     RGB Scale:Vector3f=<1 1 1>
     Alpha Scale:NiagaraFloat=&FloatFromCurve(FloatCurve=DataInterface() {propertyValues {Curve {
         defaultValue=1
         keys [
          {arriveTangent=0.5524, interpMode=RCIM_Cubic, leaveTangent=0.5524, tangentMode=RCTM_User, time=0, value=0.1}
          {arriveTangent=0.4709, interpMode=RCIM_Cubic, leaveTangent=0.4709, tangentMode=RCTM_User, time=0.24, value=1}
          {interpMode=RCIM_Cubic, tangentMode=RCTM_User, time=0.9734, value=0}
         ]
        }}})
    )
    Particles.ScaleR1:LinearColor=&VC_DynamicScaleColor(
     Color to Scale:LinearColor=$Particles.R1 Color
     RGB Scale:Vector3f=<1 1 1>
     Alpha Scale:NiagaraFloat=&FloatFromCurve(FloatCurve=DataInterface() {propertyValues {Curve {
         defaultValue=0
         keys [
          <at 0 0 RCIM_Cubic>
          {arriveTangent=2.0865, interpMode=RCIM_Cubic, leaveTangent=2.0865, tangentMode=RCTM_User, time=0.4, value=1}
          {arriveTangent=-2.8862, interpMode=RCIM_Cubic, leaveTangent=-2.8862, tangentMode=RCTM_User, time=0.85, value=0}
         ]
        }}})
    )
    Particles.ScaleR2:LinearColor=&VC_DynamicScaleColor(
     Color to Scale:LinearColor=$Particles.R2 Color
     RGB Scale:Vector3f=&VectorFromFloat(Value=&RandomRangeFloat(
       Minimum=&FloatFromCurve(FloatCurve=DataInterface() {propertyValues {Curve {
           defaultValue=1
           keys [
            {interpMode=RCIM_Cubic, tangentMode=RCTM_User, time=0, value=0.3}
            {interpMode=RCIM_Cubic, tangentMode=RCTM_User, time=0.3, value=1.5}
            {tangentMode=RCTM_User, time=0.7757, value=0.4793}
           ]
          }}})
       Maximum=&FloatFromCurve(FloatCurve=DataInterface() {propertyValues {Curve {
           defaultValue=1
           keys [
            {interpMode=RCIM_Cubic, tangentMode=RCTM_User, time=0, value=1}
            {interpMode=RCIM_Cubic, tangentMode=RCTM_User, time=0.3, value=0.5}
            {tangentMode=RCTM_User, time=1, value=1}
           ]
          }}})
      ))
     Alpha Scale:NiagaraFloat=1
    )
    Particles.WidthCore:NiagaraFloat=&Multiply_Float(
     A=40
     B=&FloatFromCurve(FloatCurve=DataInterface() {propertyValues {Curve {
         defaultValue=1
         keys [
          {interpMode=RCIM_Cubic, tangentMode=RCTM_User, time=0.05, value=0.8}
          {arriveTangent=-3.3133, arriveTangentWeight=0.6234, interpMode=RCIM_Cubic, leaveTangent=-3.3133, tangentMode=RCTM_User, time=0.17, value=0}
         ]
        }}})
    )
    Particles.WidthOutsideFire:NiagaraFloat=&Multiply_Float(
     A=35
     B=&FloatFromCurve(FloatCurve=DataInterface() {propertyValues {Curve {
         defaultValue=1
         keys [
          {arriveTangent=-6.058, interpMode=RCIM_Cubic, leaveTangent=-6.058, tangentMode=RCTM_User, time=0, value=2}
          {interpMode=RCIM_Cubic, tangentMode=RCTM_User, time=0.4, value=0.9858}
          {arriveTangent=-4.821, interpMode=RCIM_Cubic, leaveTangent=-4.821, tangentMode=RCTM_User, time=0.8, value=0.05}
         ]
        }}})
    )
   ) {isSetParameters=true}
   &ScaleRibbonWidth(Ribbon Width Scale=&FloatFromCurve(FloatCurve=DataInterface() {propertyValues {Curve {
        defaultValue=1
        keys [
         <at 0.0005 0.3673 RCIM_Cubic>
         {interpMode=RCIM_Cubic, tangentMode=RCTM_User, time=0.1014, value=1}
         {arriveTangent=-1.8676, interpMode=RCIM_Cubic, leaveTangent=-1.8676, tangentMode=RCTM_User, time=1, value=0}
        ]
       }}}))
   &DynamicMaterialParameters(
    Module.Index3Float0Write=true
    Module.Index3Float1Write=true
    Module.Index3Float2Write=true
    Module.Index3Float3Write=true
    Write Parameter Index 0=::NewEnumerator0
    Write Parameter Index 3=::NewEnumerator1
   ) {enabled=false}
  ]
  rendererClasses [&NiagaraRibbonRendererProperties]
  renderers [
   {rendererClass=&NiagaraRibbonRendererProperties}
   {rendererClass=&NiagaraRibbonRendererProperties, rendererIndex=1}
   {rendererClass=&NiagaraRibbonRendererProperties, rendererIndex=2}
   {rendererClass=&NiagaraRibbonRendererProperties, rendererIndex=3}
   {rendererClass=&NiagaraRibbonRendererProperties, rendererIndex=4}
  ]
 }
 NE_SF_WheelWrap {
  emitterUpdateScript [
   &EmitterState(
    Inactive Response=::NewEnumerator1
    Life Cycle Mode=::NewEnumerator1
    Loop Behavior=::NewEnumerator1
    Loop Duration Mode=::NewEnumerator1
    UseLoopDelay=false
   )
   &SpawnBurst_Instantaneous(Spawn Count=2, Use Spawn Probability=false, UseLoopCountLimit=false)
  ]
  particleSpawnScript [
   &VC_PaintedVariantsOverride(DefaultColor=<rgba 5 2.0363 0.2533 1>, R1 DefaultColor=<rgba 1.1458 7.4327 10 1>, Use R1 Painted Overrides=true) {enabled=false}
   &InitializeParticle(
    Color Mode=::NewEnumerator0
    Mesh Scale Mode=::NewEnumerator4
    Sprite Rotation Mode=::NewEnumerator2
    Sprite Size Mode=::NewEnumerator3
    UsePositionOffset=false
    Write Material Random=true
   )
   &PartitionParticles()
   &SkeletalMeshLocation(
    Bone / Socket Index=&Add_Integer(A=$Particles.PartitionParticles.Partition, B=4)
    Bone / Socket Sampling Mode=::NewEnumerator5
    Interpolate Mesh Sampling=false
    MaskSpawnGroup=false
    Mesh Sampling Type=::NewEnumerator6
    Skeletal Mesh=$System.SkeletalMesh
   )
  ]
  particleUpdateScript [
   &FN_DisableTireFX(
    Mask Index Attribute:NiagaraInt32=$Particles.PartitionParticles.Partition
    Mask Index FL=2
    Mask Index FR=3
    Mask Index RL=0
    Mask Index RR=1
   )
   &VC_WheelWrapScale(Wheel Index Attribute:NiagaraInt32=&Add_Integer(A=$Particles.PartitionParticles.Partition, B=2))
   &ParticleState(Kill Particles When Lifetime Has Elapsed=false)
   &UpdateMeshOrientation(Rotation Coordinate Space=::NewEnumerator2, Rotation Vector=<0 1 0>) {enabled=false}
   &SkeletalMeshLocation(
    Bone / Socket Index=&Add_Integer(A=$Particles.PartitionParticles.Partition, B=4)
    Bone / Socket Sampling Mode=::NewEnumerator5
    MaskSpawnGroup=false
    Mesh Sampling Type=::NewEnumerator6
    Skeletal Mesh=$System.SkeletalMesh
   ) {module=SkeletalMeshLocation001}
   &SolveForcesAndVelocity(Clamp Velocity=false, Limit Acceleration=false)
   &VC_PaintedVariantsOverride(
    Black=<rgba 0.003 0.003 0.003 2>
    BurntSienna=<rgba 0.05 0.0132 0.0075 2>
    Cobalt=<rgba 0.015 0.0459 0.3 1>
    Crimson=<rgba 0.22 0 0 1>
    DefaultColor=<rgba 1 0.2835 0.0595 1>
    ForestGreen=<rgba 0 0.1 0.0083 2>
    Grey=<rgba 0.0595 0.0656 0.07 1>
    Lime=<rgba 0.1114 0.2 0.0032 1>
    Orange=<rgba 0.314 0.0342 0.0031 1>
    Pink=<rgba 0.3 0.075 0.2625 1>
    Purple=<rgba 0.0463 0.015 0.25 1>
    R1 Black:LinearColor=<rgba 0.0392 0.0395 0.04 1>
    R1 BurntSienna:LinearColor=<rgba 0.25 0.0388 0.0015 1>
    R1 Cobalt:LinearColor=<rgba 0.055 0.1127 0.55 1>
    R1 Crimson:LinearColor=<rgba 0.65 0.0065 0.0066 1>
    R1 DefaultColor=<rgba 0.1009 0.6542 0.8802 1>
    R1 ForestGreen:LinearColor=<rgba 0.0917 0.44 0.022 1>
    R1 Grey:LinearColor=<rgba 0.17 0.17 0.17 1>
    R1 Lime:LinearColor=<rgba 0.4 0.8 0 1>
    R1 Orange:LinearColor=<rgba 0.6 0.1644 0.006 1>
    R1 Pink:LinearColor=<rgba 0.75 0.1875 0.6094 1>
    R1 Purple:LinearColor=<rgba 0.1432 0.0234 0.78 1>
    R1 Saffron:LinearColor=<rgba 2 1.3333 0 1>
    R1 SkyBlue:LinearColor=<rgba 0.1 0.6695 1 1>
    R1 TitaniumWhite:LinearColor=<rgba 0.9 0.9 0.9 1>
    Saffron=<rgba 0.45 0.315 0 1>
    SkyBlue=<rgba 0.03 0.21 0.3 1>
    TitaniumWhite=<rgba 0.2 0.1967 0.18 1>
    Use R1 Painted Overrides=true
   ) {module=VC_PaintedVariantsOverride001}
  ]
  rendererClasses [&NiagaraMeshRendererProperties]
  renderers [
   {rendererClass=&NiagaraMeshRendererProperties, rendererIndex=0}
   {rendererClass=&NiagaraMeshRendererProperties, rendererIndex=1}
  ]
 }
}
summary {
 systemName=NS_DriftTrail_Slimefire
 emitters [
  {emitterName=NE_SF_Sphere, rendererClasses [&NiagaraSpriteRendererProperties, &NiagaraMeshRendererProperties]}
  {emitterName=NE_SF_Trail, rendererClasses [&NiagaraRibbonRendererProperties]}
  {emitterName=NE_SF_Sprite, rendererClasses [&NiagaraSpriteRendererProperties]}
  {emitterName=NE_SF_WheelWrap, rendererClasses [&NiagaraMeshRendererProperties]}
 ]
 userVariables [
  User.bIsFrontendPreview {type=NiagaraBool}
  User.bIsPainted {type=NiagaraBool}
  User.ColorVariantID {type=NiagaraInt32}
  User.DTCarOffset {type=Vector3f, description="", defaultValue {struct=Vector3f, value=<0 0 0>}}
  User.offset {type=Vector3f, description="offset from wheels(set through cosmetic template)", defaultValue {struct=Vector3f, value=<0 0 0>}}
  User.PaintedColor {type=LinearColor, description="", defaultValue {struct=LinearColor, value=<rgba 1 1 1 1>}}
  User.PaintedScalar {type=NiagaraFloat, defaultValue=1, description=""}
  User.VehicleFlags {type=NiagaraInt32, defaultValue=806289408}
  User.Wheel_Radius_Array {
   type=NiagaraDataInterfaceArrayFloat
   defaultValue {
    struct=NiagaraExt_VariableValue_DataInterface
    value {
     dataInterface=&self:NiagaraDataInterfaceArrayFloat_1
     dataInterfaceClass=&NiagaraDataInterfaceArrayFloat
    }
   }
  }
  User.Wheel_Width_Array {
   type=NiagaraDataInterfaceArrayFloat
   defaultValue {
    struct=NiagaraExt_VariableValue_DataInterface
    value {
     dataInterface=&self:NiagaraDataInterfaceArrayFloat_0
     dataInterfaceClass=&NiagaraDataInterfaceArrayFloat
    }
   }
  }
  User.WheelTouchingArray {
   type=NiagaraDataInterfaceArrayFloat
   defaultValue {
    struct=NiagaraExt_VariableValue_DataInterface
    value {
     dataInterface=&self:NiagaraDataInterfaceArrayFloat_2
     dataInterfaceClass=&NiagaraDataInterfaceArrayFloat
    }
   }
  }
 ]
}
